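; Shallow-water fluid simulation Blueprint that adds interactive ripples to water bodies, gated by water scalability and cosmetic-only execution.
; Impulse, dynamic and projectile forces are canvas-drawn into a forces target using cached MIDs.
; A ripple material cycles three render targets with offsets compensating for grid movement, then a normal target is rendered for water materials.
; Fixed time stepping is switchable, and simulation idles after 600 force-free frames.
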(event Custom|FluidGridDebug
  (bind _returnvalue (Rendering|Material|GetScalarParameterValue "/Water/Materials/MPC/MPC_Water.MPC_Water" "DebugSimGrid" :!class /Script/Engine.KismetMaterialLibrary))
  (Rendering|Material|SetScalarParameterValue "/Water/Materials/MPC/MPC_Water.MPC_Water" "DebugSimGrid" (select (== _returnvalue 1.0) 0.0 1.0) :!class /Script/Engine.KismetMaterialLibrary))

(event Custom|Release%20RTs
  (for (_ _array_index) (Variables|Debug|GetRippleRTs)
    (Utilities|Array|SetArrayElem (Variables|Debug|GetRippleRTs) _array_index))
  (Variables|Debug|SetNormalRT "")
  (Variables|Debug|SetForcesRT "")
  (Variables|Debug|SetPerfRT ""))

(event Custom|Allocate%20RTs
  (bind _returnvalue (Water|GetShallowWaterSimulationRenderTargetSize))
  (Variables|Simulation|SetResolution _returnvalue)
  (for (_array_element _array_index) (Variables|Debug|GetRippleRTs)
    (Variables|Default|SetTempRT _array_element)
    (Utilities|CreateOrUpdateRT (Variables|Default|GetTempRT) (Utilities|Struct|MakeIntPoint (Variables|Simulation|GetResolution) (Variables|Simulation|GetResolution))
      (:Out
        (Utilities|Array|SetArrayElem (Variables|Debug|GetRippleRTs) _array_index (Variables|Default|GetTempRT))
        (Variables|Default|SetTempRT ""))))
  (Utilities|CreateOrUpdateRT (Variables|Debug|GetNormalRT) (Utilities|Struct|MakeIntPoint (Variables|Simulation|GetResolution) (Variables|Simulation|GetResolution)) true
    (:Out
      (Utilities|CreateOrUpdateRT (Variables|Debug|GetForcesRT) (Utilities|Struct|MakeIntPoint (Variables|Simulation|GetResolution) (Variables|Simulation|GetResolution)) true
        (:Out
          (if (Variables|Debug|GetPerf%20Test%20Mode)
            (Utilities|CreateOrUpdateRT (Variables|Debug|GetPerfRT) (Utilities|Struct|MakeIntPoint 4 4) true)))))))

(event Custom|Clear%20RTs
  ;[0.13 0.13 0.13] Clear with Black
  (for _array_element (Variables|Debug|GetRippleRTs)
    (Rendering|ClearRenderTarget2D _array_element))
  (Rendering|ClearRenderTarget2D (Variables|Debug|GetNormalRT) "(R=0.000000,G=0.000000,B=0.500000,A=1.000000)")
  (Rendering|ClearRenderTarget2D (Variables|Debug|GetForcesRT)))

(event Custom|ReCheckScalability
  (bind _waterzone (Variables|Simulation|GetWaterZone))
  (bind _output (Utilities|ResolveSoftReference _waterzone))
  (bind _waterzone_2 (Variables|Simulation|GetWaterZone))
  (bind _output_2 (Utilities|ResolveSoftReference _waterzone_2))
  (if (Water|IsShallowWaterSimulationEnabled (WorldSubsystems|GetWaterSubsystem))
    (Networking|CanExecuteCosmeticEvents
      (:True
        (CallFunction|AllocateRTs)
        (CallFunction|ClearRTs)
        (if (Variables|Simulation|GetUse%20Terrain%20Water%20System)
          ; Get Landscape WaterZone Info
          (Variables|Debug|SetMaxFlowVelocity (Fortnite|GetNumericCVar "r.Water.MaxFlowVelocity"))
          (if (Utilities|IsValidSoftObjectReference (Variables|Simulation|GetWaterZone))
            (Utilities|IsValid _output_2
              (:Is%20Valid
                (Utilities|IsValid (Variables|/Script/Water.WaterZone|GetWaterInfoTextureArray _output_2)
                  (:Is%20Valid
                    (Variables|Simulation|SetWaterVelocityTexture (Variables|/Script/Water.WaterZone|GetWaterInfoTextureArray (Utilities|ResolveSoftReference (Variables|Simulation|GetWaterZone))))
                    (Variables|Debug|SetWaterZoneIndex (Water|GetWaterZoneIndex (Utilities|ResolveSoftReference (Variables|Simulation|GetWaterZone))))
                    (Variables|Debug|SetWaterZoneExtent (Variables|/Script/Water.WaterZone|GetZoneExtent _output))
                    (Variables|Debug|SetWaterZoneLocation (Transformation|GetActorLocation _output)))
                  (:Is%20Not%20Valid
                    (Water|BindEventtoOnWaterInfoTextureArrayCreated (Utilities|ResolveSoftReference (Variables|Simulation|GetWaterZone)) (delegate Custom|WaterZoneTextureCreated)))))
              (:Is%20Not%20Valid
                (Development|PrintString "BP_FluidSim: Warning! Sim set to 'Use Terrain Water System', but WaterZone reference failed to Resolve SoftObjectReference to Valid WaterZone.")))
            (else
              (Development|PrintString "BP_FluidSim: Warning! Sim set to 'Use Terrain Water System', but WaterZone reference was Invalid."))))
        (Default|SetupDisplayMIDs)
        (Default|SetupSimMIDs)
        (CallFunction|SetWaterbodyMIDParams)
        (bind _rts_all_valid (CallFunction|ValidateRTs))
        (if _rts_all_valid
          (Variables|Simulation|SetEnabled true)
          (Water|PrinttoWaterLog (WorldSubsystems|GetWaterSubsystem) (Utilities|String|Append "WATER LOG: " "All RTs validated succesfully.") false)
          (else
            (Variables|Simulation|SetEnabled false)))))
    (else
      (CallFunction|ReleaseRTs)
      (Variables|Simulation|SetEnabled false)
      (CallFunction|ClearSimFromWaterbodyMIDs))))

(event EventBeginPlay
  (bind _waterzone (Variables|Simulation|GetWaterZone))
  (bind _output (Utilities|ResolveSoftReference _waterzone))
  (bind _waterzone_2 (Variables|Simulation|GetWaterZone))
  (bind _output_2 (Utilities|ResolveSoftReference _waterzone_2))
  (bind _returnvalue (WorldSubsystems|GetWaterSubsystem))
  (Water|BindEventtoOnWaterScalabilityChanged _returnvalue (delegate Custom|ReCheckScalability))
  (if (Water|IsShallowWaterSimulationEnabled _returnvalue)
    (Networking|CanExecuteCosmeticEvents
      (:True
        (CallFunction|AllocateRTs)
        (CallFunction|ClearRTs)
        (if (Variables|Simulation|GetUse%20Terrain%20Water%20System)
          ; Get Landscape WaterZone Info
          (Variables|Debug|SetMaxFlowVelocity (Fortnite|GetNumericCVar "r.Water.MaxFlowVelocity"))
          (if (Utilities|IsValidSoftObjectReference (Variables|Simulation|GetWaterZone))
            (Utilities|IsValid _output_2
              (:Is%20Valid
                (Utilities|IsValid (Variables|/Script/Water.WaterZone|GetWaterInfoTextureArray _output_2)
                  (:Is%20Valid
                    (Variables|Simulation|SetWaterVelocityTexture (Variables|/Script/Water.WaterZone|GetWaterInfoTextureArray (Utilities|ResolveSoftReference (Variables|Simulation|GetWaterZone))))
                    (Variables|Debug|SetWaterZoneIndex (Water|GetWaterZoneIndex (Utilities|ResolveSoftReference (Variables|Simulation|GetWaterZone))))
                    (Variables|Debug|SetWaterZoneExtent (Variables|/Script/Water.WaterZone|GetZoneExtent _output))
                    (Variables|Debug|SetWaterZoneLocation (Transformation|GetActorLocation _output)))
                  (:Is%20Not%20Valid
                    (Water|BindEventtoOnWaterInfoTextureArrayCreated (Utilities|ResolveSoftReference (Variables|Simulation|GetWaterZone)) (delegate Custom|WaterZoneTextureCreated)))))
              (:Is%20Not%20Valid
                (Development|PrintString "BP_FluidSim: Warning! Sim set to 'Use Terrain Water System', but WaterZone reference failed to Resolve SoftObjectReference to Valid WaterZone.")))
            (else
              (Development|PrintString "BP_FluidSim: Warning! Sim set to 'Use Terrain Water System', but WaterZone reference was Invalid."))))
        (Default|SetupDisplayMIDs)
        (Default|SetupSimMIDs)
        (CallFunction|SetWaterbodyMIDParams)
        (bind _rts_all_valid (CallFunction|ValidateRTs))
        (if _rts_all_valid
          (Variables|Simulation|SetEnabled true)
          (Water|PrinttoWaterLog (WorldSubsystems|GetWaterSubsystem) (Utilities|String|Append "WATER LOG: " "All RTs validated succesfully.") false)
          (else
            (Variables|Simulation|SetEnabled false)))))))

(event EventTick
  (params (DeltaSeconds :type Float%20%28single-precision%29))
  (bind _normalrt (Variables|Debug|GetNormalRT))
  (if (Variables|Simulation|GetEnabled)
    (CallFunction|GetPlayerPawnForces)
    (CallFunction|GetFramesSinceLastActiveForces)
    (if (< (Variables|Default|GetFrames%20Since%20Last%20Active%20Force) 600)
      (Variables|Debug|SetGridCenter (Utilities|GetGridCenter))
      (Utilities|FlowControl|Sequence
        (:then_0
          ;[1 1 1] Not need for simulation, useful for gathing accurate perf stats via profilegpu
          (if (Variables|Debug|GetPerf%20Test%20Mode)
            (Rendering|DrawMaterialtoRenderTarget (Variables|Debug|GetPerfRT) "/FortWater/FluidSimulation/Materials/Debug/PerfTestNoiseMaterial.PerfTestNoiseMaterial")))
        (:then_1
          ;[1 0.05729 0.05729] Debug For Jittery Sim
          (if (Variables|Debug|GetUpdateForcesInFixedTimeStep)
            _
            (else
              _)))
        (:then_2
          ;[0.5938 0.05876 0.05876] Debug
          (if (Variables|Debug|GetUpdateSimInFixedTimeStep)
            (Utilities|FixedTimeStep (Variables|Default|GetFixedStep) 3
              (:Fixed%20Step%20Execution
                (Utilities|FlowControl|Sequence
                  (:then_0
                    (if (Variables|Debug|GetUpdateForcesInFixedTimeStep)
                      ;[1 1 1] Physics Forces
                      (CallFunction|UpdateDynamicForces :!label _label_2)
                      (if (Variables|Simulation|GetAdd%20Physics%20Forces)
                        (Rendering|DrawMaterialtoRenderTarget (Variables|Debug|GetForcesRT) (Variables|Debug|GetPhysicsForceMID)))
                      ;[1 0.7805 0.09399] Impulse Forces
                      (bind (_canvas _size _context) (Rendering|BeginDrawCanvastoRenderTarget (Variables|Debug|GetForcesRT)))
                      (for _array_element (Variables|Forces|GetImpulseForces)
                        (Default|DrawImpulseForce _canvas _size _array_element))
                      (Rendering|EndDrawCanvastoRenderTarget _context)
                      ;[0.1995 1 0.2653] Dynamic Forces
                      (bind (_canvas_2 _size_2 _context_2) (Rendering|BeginDrawCanvastoRenderTarget (Variables|Debug|GetForcesRT)))
                      (bind _values (Utilities|Map|Values (Variables|Forces|GetDynamicForces)))
                      (for _array_element_2 _values
                        (Default|DrawDynamicForce _canvas_2 _size_2 _array_element_2))
                      (Rendering|EndDrawCanvastoRenderTarget _context_2)
                      ;[1 0 0] Projectile Forces
                      (bind (_canvas_3 _size_3 _context_3) (Rendering|BeginDrawCanvastoRenderTarget (Variables|Debug|GetForcesRT)))
                      (bind _values_2 (Utilities|Map|Values (Variables|Forces|GetProjectileForces)))
                      (for _array_element_3 _values_2
                        (Default|DrawDynamicForce _canvas_3 _size_3 _array_element_3))
                      (Rendering|EndDrawCanvastoRenderTarget _context_3)))
                  (:then_1
                    _)
                  (:then_2
                    ;[1 0.05729 0.05729] Debug For Jittery Sim
                    (if (Variables|Debug|GetUpdateNormalInFixedTimeStep)
                      ;[1 1 1] Normal RT Update
                      (Rendering|ClearRenderTarget2D _normalrt :!label _label)
                      (Rendering|DrawMaterialtoRenderTarget _normalrt (Variables|Debug|GetRenderNormalsMID))
                      (CallFunction|UpdateImpulseLifetimes)))))))
          ;[0.2266 0.6487 1] Fluid Simulation
          (CallFunction|GridMovement)
          (bind _current_target (CallFunction|CycleRenderTargets))
          ;[1 0.02604 0.03844] Debug RT Cycle
          (if (Variables|Debug|GetShowDebugWaterPlane)
            (CallFunction|CycleDebugRenderTargets))
          (Rendering|DrawMaterialtoRenderTarget _current_target (Variables|Debug|GetRippleSimMID))
          (Math|Integer|IncrementInt (Variables|Default|GetPasses)))
        (:then_3
          (if (not (Variables|Debug|GetUpdateNormalInFixedTimeStep))
            (goto _label))))
      (goto _label_2)
      (else
        (Math|Integer|IncrementInt (Variables|Default|GetPasses))))))

(event Custom|SetShowDebugWaterPlane
  (params (IsVisible :type Boolean))
  (bind _debugwaterplanemid (Variables|Debug|GetDebugWaterPlaneMID))
  ;[1 0.08281 0.07286] Debug Functions
  (bind _output_get (Variables|Debug|SetShowDebugWaterPlane IsVisible))
  (Development|SetHiddeninGame (Variables|Default|GetDebugPlane) (not _output_get) false)
  (if IsVisible
    (Transformation|SetWorldScale3D (Variables|Default|GetDebugPlane) (Math|Conversions|ToVector(Float) (/ (Variables|Simulation|GetFluid%20Size) 1e+02)))
    (Rendering|Material|SetTextureParameterValue _debugwaterplanemid "Texture" (Variables|Debug|GetForcesRT))
    (Rendering|Material|SetScalarParameterValue _debugwaterplanemid "DisplayNormal" 0.0 :!class /Script/Engine.MaterialInstanceDynamic)))

(event Custom|SetDebugNormalRT
  (bind _debugwaterplanemid (Variables|Debug|GetDebugWaterPlaneMID))
  ;[1 0.08281 0.07286] Debug Functions
  (Rendering|Material|SetTextureParameterValue _debugwaterplanemid "Normal" (Variables|Debug|GetNormalRT))
  (Rendering|Material|SetScalarParameterValue _debugwaterplanemid "DisplayNormal" 1.0 :!class /Script/Engine.MaterialInstanceDynamic))

(event Custom|SetDebugRippleRTCurrent
  (bind _debugwaterplanemid (Variables|Debug|GetDebugWaterPlaneMID))
  ;[1 0.08281 0.07286] Debug Functions
  (Rendering|Material|SetTextureParameterValue _debugwaterplanemid "Texture" (Variables|Debug|GetDebugRippleRTCurrent))
  (Rendering|Material|SetScalarParameterValue _debugwaterplanemid "DisplayNormal" 0.0 :!class /Script/Engine.MaterialInstanceDynamic))

(event Custom|SetDebugRippleRT1
  (bind _debugwaterplanemid (Variables|Debug|GetDebugWaterPlaneMID))
  ;[1 0.08281 0.07286] Debug Functions
  (Rendering|Material|SetTextureParameterValue _debugwaterplanemid "Texture" (Variables|Debug|GetDebugRippleRT1))
  (Rendering|Material|SetScalarParameterValue _debugwaterplanemid "DisplayNormal" 0.0 :!class /Script/Engine.MaterialInstanceDynamic))

(event Custom|SetUpdateForcesInFixedTimeStep
  (params (Bool :type Boolean))
  ;[1 0.08281 0.07286] Debug Functions
  (Variables|Debug|SetUpdateForcesInFixedTimeStep Bool))

(event Custom|SetDebugRippleRT2
  (bind _debugwaterplanemid (Variables|Debug|GetDebugWaterPlaneMID))
  ;[1 0.08281 0.07286] Debug Functions
  (Rendering|Material|SetTextureParameterValue _debugwaterplanemid "Texture" (Variables|Debug|GetDebugRippleRT2))
  (Rendering|Material|SetScalarParameterValue _debugwaterplanemid "DisplayNormal" 0.0 :!class /Script/Engine.MaterialInstanceDynamic))

(event Custom|SetDebugForcesRT
  (bind _debugwaterplanemid (Variables|Debug|GetDebugWaterPlaneMID))
  ;[1 0.08281 0.07286] Debug Functions
  (Rendering|Material|SetTextureParameterValue _debugwaterplanemid "Texture" (Variables|Debug|GetForcesRT))
  (Rendering|Material|SetScalarParameterValue _debugwaterplanemid "DisplayNormal" 0.0 :!class /Script/Engine.MaterialInstanceDynamic))

(event Custom|SetNormalInFixedTimeStep
  (params (Bool :type Boolean))
  ;[1 0.08281 0.07286] Debug Functions
  (Variables|Debug|SetUpdateNormalInFixedTimeStep Bool))

(event Custom|SetSimInFixedTimeStep
  (params (Bool :type Boolean))
  ;[1 0.08281 0.07286] Debug Functions
  (Variables|Debug|SetUpdateSimInFixedTimeStep Bool))

(event Custom|SetAllSimParamsInFixedTimeStep
  (params (Bool :type Boolean))
  ;[1 0.08281 0.07286] Debug Functions
  (Variables|Debug|SetUpdateForcesInFixedTimeStep Bool)
  (Variables|Debug|SetUpdateSimInFixedTimeStep Bool)
  (Variables|Debug|SetUpdateNormalInFixedTimeStep Bool))

(fn ApplyFluidForceImpulse
  (params (Impulse%20Settings :type Fluid%20Force%20Impulse%20Structure))
  (bind (_worldposition_32_c9a94c02422d8bf40df6b1bb2a0d8cbc _timedwaterdropsplashes_38_2cb1456b483aef3a52677aaf4152e31e) (Utilities|Struct|BreakFluidForceImpulse Impulse%20Settings))
  (bind (_enablewaterdropseffect_39_0a7932284406807d62695d8e0927bd70 _effectradius_30_c9a94c02422d8bf40df6b1bb2a0d8cbc _strength_29_2caa30794d1eff60ae1c3491d011cecf _waterdropspersquaremeter_27_c31b527c4c367a5ca5e1df8e49e76234 _starttimeoffset_33_5a792ce8489a59e5d9b24f9e4dcfe94a _lifetime_35_c2749c1449c41d4f236bcbaf6ed34113 _strengthoverlifepower_37_4fa6941a4ad024828afeb782783dd01c _materialoverride_42_fb856a244a1713590bb76eaa7cc7a0df) (Utilities|Struct|BreakFluidForceImpulseTimedEffects _timedwaterdropsplashes_38_2cb1456b483aef3a52677aaf4152e31e))
  (Networking|CanExecuteCosmeticEvents
    (:True
      (bind _returnvalue (Water|GetShallowWaterMaxImpulseForces))
      (bind _output_get (LocalVariables|SetMaxImpulseForces _returnvalue))
      (if (< (Utilities|Array|Length (Variables|Forces|GetImpulseForces)) _output_get)
        (if (<= (Math|Vector|DotProduct (* (- _worldposition_32_c9a94c02422d8bf40df6b1bb2a0d8cbc (Variables|Debug|GetGridCenter)) "1.000000,1.000000, 0") (* (- _worldposition_32_c9a94c02422d8bf40df6b1bb2a0d8cbc (Variables|Debug|GetGridCenter)) "1.000000,1.000000, 0")) (Variables|Default|GetFluidSizeSquared))
          (Utilities|Array|Add (Variables|Forces|GetImpulseForces) (Utilities|Struct|MakeFluidForceImpulsePerInstanceData Impulse%20Settings))
          (if (and (< (Utilities|Array|Length (Variables|Forces|GetImpulseForces)) (LocalVariables|GetMaxImpulseForces)) _enablewaterdropseffect_39_0a7932284406807d62695d8e0927bd70)
            (Utilities|OptionalDebugString "Add Splash Effect to Impulse Forces"
              (:Out
                (Utilities|Array|Add (Variables|Forces|GetImpulseForces) (Utilities|Struct|MakeFluidForceImpulsePerInstanceData (Utilities|Struct|MakeFluidForceImpulse _worldposition_32_c9a94c02422d8bf40df6b1bb2a0d8cbc _effectradius_30_c9a94c02422d8bf40df6b1bb2a0d8cbc _strength_29_2caa30794d1eff60ae1c3491d011cecf (select (Utilities|IsValid _materialoverride_42_fb856a244a1713590bb76eaa7cc7a0df :!class /Script/Engine.KismetSystemLibrary) _materialoverride_42_fb856a244a1713590bb76eaa7cc7a0df "/FortWater/FluidSimulation/Materials/Effects/M_Rain_01_Inst.M_Rain_01_Inst") _lifetime_35_c2749c1449c41d4f236bcbaf6ed34113 _strengthoverlifepower_37_4fa6941a4ad024828afeb782783dd01c _timedwaterdropsplashes_38_2cb1456b483aef3a52677aaf4152e31e) 0.0 _starttimeoffset_33_5a792ce8489a59e5d9b24f9e4dcfe94a true)))))
          (else
            (Utilities|OptionalDebugString "Impulse Ignored because it was too far from the Fluid Simulation.")))))))

(fn ClearSimfromWaterbodyMIDs
  (bind _outactors (Actor|GetAllActorsOfClass "/Script/Water.WaterBody"))
  (for _array_element _outactors
    (bind _returnvalue (Rendering|GetWaterMaterialInstance (Water|GetWaterBodyComponent _array_element)))
    (Rendering|Material|SetTextureParameterValue _returnvalue "NormalAndHeight" "/Engine/EngineResources/Black.Black")
    (Rendering|Material|SetTextureParameterValue _returnvalue "Forces" "/Engine/EngineResources/Black.Black")
    (bind _returnvalue_2 (Rendering|GetRivertoOceanTransitionMaterialInstance (Water|GetWaterBodyComponent _array_element)))
    (Utilities|IsValid _returnvalue_2
      (:Is%20Valid
        (Rendering|Material|SetTextureParameterValue _returnvalue_2 "NormalAndHeight" "/Engine/EngineResources/Black.Black"))
      (:Is%20Not%20Valid
        _))
    (bind _returnvalue_3 (Rendering|GetRivertoLakeTransitionMaterialInstance (Water|GetWaterBodyComponent _array_element)))
    (Utilities|IsValid _returnvalue_3
      (:Is%20Valid
        (Rendering|Material|SetTextureParameterValue _returnvalue_3 "NormalAndHeight" "/Engine/EngineResources/Black.Black")))))

(fn ConvertForcePosition
  (params (Force%20Location :type Vector) (Sine%20Bob :type Float%20%28double-precision%29))
  ;[1 1 1] Force Pos
  (return (+ (Math|Vector|MakeVector (.x (* (+ (/ (- Force%20Location (Variables|Debug|GetGridCenter)) (* (Variables|Simulation|GetFluid%20Size) 0.5)) 1.0) 0.5)) (.y (* (+ (/ (- Force%20Location (Variables|Debug|GetGridCenter)) (* (Variables|Simulation|GetFluid%20Size) 0.5)) 1.0) 0.5)) (/ (.z (- Force%20Location (Variables|Debug|GetGridCenter))) (Variables|Simulation|GetFluid%20Size))) (Math|Vector|MakeVector 0.0 (* (Math|Trig|Sin(Radians) (* (Utilities|Time|GetGameTimeinSeconds) 1)) Sine%20Bob)))))

(fn CycleRenderTargets
  (bind _ripplesimmid (Variables|Debug|GetRippleSimMID))
  (bind _passes (Variables|Default|GetPasses))
  (Rendering|Material|SetTextureParameterValue _ripplesimmid "RT" (Utilities|Array|Get(acopy) (Variables|Debug|GetRippleRTs) (Math|Integer|%25%28Integer%29 (+ _passes 2) 3)))
  (Rendering|Material|SetTextureParameterValue _ripplesimmid "RT2" (Utilities|Array|Get(acopy) (Variables|Debug|GetRippleRTs) (Math|Integer|%25%28Integer%29 (+ _passes 1) 3)))
  (return (Utilities|Array|Get(acopy) (Variables|Debug|GetRippleRTs) (Math|Integer|%25%28Integer%29 _passes 3))))

(fn CycleDebugRenderTargets
  (bind _passes (Variables|Default|GetPasses))
  (bind _ripplerts (Variables|Debug|GetRippleRTs))
  (Variables|Debug|SetDebugRippleRT1 (Utilities|Array|Get(acopy) _ripplerts (Math|Integer|%25%28Integer%29 (+ _passes 2) 3)))
  (Variables|Debug|SetDebugRippleRT2 (Utilities|Array|Get(acopy) _ripplerts (Math|Integer|%25%28Integer%29 (+ _passes 1) 3)))
  (Variables|Debug|SetDebugRippleRTCurrent (Utilities|Array|Get(acopy) _ripplerts (Math|Integer|%25%28Integer%29 _passes 3)))
  (return))

(fn DrawDynamicForce
  (params (Canvas :type Canvas%20Object%20Reference%20%28by%20ref%29) (Canvas%20Size :type Vector%202D%20Structure%20%28by%20ref%29) (Dynamic%20Force%20Settings :type Fluid%20Force%20Dynamic%20Per%20Instance%20Data%20Structure%20%28by%20ref%29))
  (bind _localforce (LocalVariables|GetLocalForce))
  (bind (_forceinfo_2_a6a35e3243faf59d161a5fbaa2f6f2b1 _componentlocation_13_959307184c8e5ccaca55fc8378d92cfd _componentvelocity_5_4f6589474918826df8a6468cf0f2c361 _boundsradius_30_adff818743be39ac4a481d995cb50d03 _waterlevel_34_a4e505d148073b883ca7b1b09a3e34a8) (Utilities|Struct|BreakFluidForceDynamicPerInstanceData _localforce))
  (bind _localforce_2 (LocalVariables|GetLocalForce))
  (bind _forceinfo_2_a6a35e3243faf59d161a5fbaa2f6f2b1_2 (Utilities|Struct|BreakFluidForceDynamicPerInstanceData _localforce_2))
  (bind _forcestrength_33_2caa30794d1eff60ae1c3491d011cecf (Utilities|Struct|BreakFluidForceDynamic _forceinfo_2_a6a35e3243faf59d161a5fbaa2f6f2b1_2))
  (bind (_forceinfo_2_a6a35e3243faf59d161a5fbaa2f6f2b1_3 _waterlevel_34_a4e505d148073b883ca7b1b09a3e34a8_2) (Utilities|Struct|BreakFluidForceDynamicPerInstanceData (LocalVariables|GetLocalForce)))
  (bind (_forcestrength_33_2caa30794d1eff60ae1c3491d011cecf_2 _forcecomponent_34_abf6640f40d37677ef6f809e09046055) (Utilities|Struct|BreakFluidForceDynamic _forceinfo_2_a6a35e3243faf59d161a5fbaa2f6f2b1))
  (bind _localforce_3 (LocalVariables|GetLocalForce))
  (bind (_forceinfo_2_a6a35e3243faf59d161a5fbaa2f6f2b1_4 _waterlevel_34_a4e505d148073b883ca7b1b09a3e34a8_3 _socketlocationmap_21_abf6aa244a5f84728a5e83be2328c7fa _socketvelocitymap_26_82b0e24b45935a12e1949f918a59a537) (Utilities|Struct|BreakFluidForceDynamicPerInstanceData _localforce_3))
  (bind (_componentlocation_13_959307184c8e5ccaca55fc8378d92cfd_2 _boundsradius_30_adff818743be39ac4a481d995cb50d03_2) (Utilities|Struct|BreakFluidForceDynamicPerInstanceData (LocalVariables|GetLocalForce)))
  (bind (_forceinfo_2_a6a35e3243faf59d161a5fbaa2f6f2b1_5 _componentlocation_13_959307184c8e5ccaca55fc8378d92cfd_3) (Utilities|Struct|BreakFluidForceDynamicPerInstanceData Dynamic%20Force%20Settings))
  (bind (_forcetype_28_ddc16ee543d2dfd3ba29c49d32198c9c _skeletalmeshsetup_31_51a4130440baffba1da0fe83e942a30a) (Utilities|Struct|BreakFluidForceDynamic _forceinfo_2_a6a35e3243faf59d161a5fbaa2f6f2b1_5))
  (bind _socketsandendpoints_6_b3edd8fc43a7c681151f46be0aa158c5 (Utilities|Struct|BreakFluidForceSocketInfo _skeletalmeshsetup_31_51a4130440baffba1da0fe83e942a30a))
  (bind (_screen_position _screen_size) (Class|RendertoTextureFunctionLibrary|SetCanvasMaterialScaleAndPosition Canvas%20Size (Math|Conversions|ToVector2D(Vector) (CallFunction|ConvertForcePosition _componentlocation_13_959307184c8e5ccaca55fc8378d92cfd)) (/ _boundsradius_30_adff818743be39ac4a481d995cb50d03 (* (Variables|Simulation|GetFluid%20Size) 0.5))))
  (bind (_screen_position_2 _screen_size_2) (Class|RendertoTextureFunctionLibrary|SetCanvasMaterialScaleAndPosition Canvas%20Size (Math|Conversions|ToVector2D(Vector) (CallFunction|ConvertForcePosition _componentlocation_13_959307184c8e5ccaca55fc8378d92cfd_2)) (/ (* _boundsradius_30_adff818743be39ac4a481d995cb50d03_2 1.1) (* (Variables|Simulation|GetFluid%20Size) 0.5))))
  (bind _returnvalue (Transformation|GetWorldRotation _forcecomponent_34_abf6640f40d37677ef6f809e09046055))
  (bind (_x _y _z) (Math|Rotator|GetAxes _returnvalue))
  (bind _localmid (LocalVariables|GetLocalMID))
  (bind _localmid_2 (LocalVariables|GetLocalMID))
  (bind _localmid_3 (LocalVariables|GetLocalMID))
  (LocalVariables|SetLocalForce Dynamic%20Force%20Settings)
  (bind _mid (Default|GetForceMID (Utilities|Struct|BreakFluidForceDynamic _forceinfo_2_a6a35e3243faf59d161a5fbaa2f6f2b1_5)))
  (LocalVariables|SetLocalMID _mid)
  (if (< (Math|Vector|DotProduct (* (- _componentlocation_13_959307184c8e5ccaca55fc8378d92cfd_3 (Variables|Debug|GetGridCenter)) "1.000000,1.000000, 0") (* (- _componentlocation_13_959307184c8e5ccaca55fc8378d92cfd_3 (Variables|Debug|GetGridCenter)) "1.000000,1.000000, 0")) (Variables|Default|GetFluidSizeSquared))
    (switch Utilities|FlowControl|Switch|SwitchonFluidDynamicForceMeshType _forcetype_28_ddc16ee543d2dfd3ba29c49d32198c9c
      (:NewEnumerator0
        (Rendering|Material|SetVectorParameterValue _localmid_3 "P" (Math|Conversions|ToLinearColor(Vector) _componentlocation_13_959307184c8e5ccaca55fc8378d92cfd) :!class /Script/Engine.MaterialInstanceDynamic)
        (Rendering|Material|SetScalarParameterValue _localmid_3 "F" (* (Math|Vector|VectorLength _componentvelocity_5_4f6589474918826df8a6468cf0f2c361) _forcestrength_33_2caa30794d1eff60ae1c3491d011cecf_2) :!class /Script/Engine.MaterialInstanceDynamic)
        (Rendering|Material|SetScalarParameterValue _localmid_3 "R" _boundsradius_30_adff818743be39ac4a481d995cb50d03 :!class /Script/Engine.MaterialInstanceDynamic)
        (Rendering|Material|SetScalarParameterValue _localmid_3 "WaterZ" _waterlevel_34_a4e505d148073b883ca7b1b09a3e34a8 :!class /Script/Engine.MaterialInstanceDynamic)
        (Utilities|IsValid _forcecomponent_34_abf6640f40d37677ef6f809e09046055
          (:Is%20Valid
            (Rendering|Material|SetVectorParameterValue _localmid "X" (Math|Conversions|ToLinearColor(Vector) _x) :!class /Script/Engine.MaterialInstanceDynamic)
            (Rendering|Material|SetVectorParameterValue _localmid "Y" (Math|Conversions|ToLinearColor(Vector) _y) :!class /Script/Engine.MaterialInstanceDynamic)
            (Rendering|Material|SetVectorParameterValue _localmid "Z" (Math|Conversions|ToLinearColor(Vector) _z) :!class /Script/Engine.MaterialInstanceDynamic)
            (Rendering|Material|SetVectorParameterValue _localmid "GridLocation" (Math|Conversions|ToLinearColor(Vector) (Variables|Debug|GetGridCenter)) :!class /Script/Engine.MaterialInstanceDynamic))
          (:Is%20Not%20Valid
            _))
        (Canvas|DrawMaterial :self Canvas :RenderMaterial (LocalVariables|GetLocalMID) :ScreenPosition _screen_position :ScreenSize _screen_size :CoordinateSize "(X=1.000000,Y=1.000000)" :Rotation 0.0 :PivotPoint "(X=0.500000,Y=0.500000)"))
      (:NewEnumerator1
        (bind _keys (Utilities|Map|Keys _socketsandendpoints_6_b3edd8fc43a7c681151f46be0aa158c5))
        (for (_array_element _array_index) _keys
          (if (and (and (.ReturnValue (Utilities|Map|Find _socketlocationmap_21_abf6aa244a5f84728a5e83be2328c7fa _array_element)) (.ReturnValue (Utilities|Map|Find _socketlocationmap_21_abf6aa244a5f84728a5e83be2328c7fa (Utilities|Map|Find _socketsandendpoints_6_b3edd8fc43a7c681151f46be0aa158c5 _array_element)))) (and (.ReturnValue (Utilities|Map|Find _socketvelocitymap_26_82b0e24b45935a12e1949f918a59a537 _array_element)) (.ReturnValue (Utilities|Map|Find _socketvelocitymap_26_82b0e24b45935a12e1949f918a59a537 (Utilities|Map|Find _socketsandendpoints_6_b3edd8fc43a7c681151f46be0aa158c5 _array_element)))))
            (Rendering|Material|SetVectorParameterValue (LocalVariables|GetLocalMID) (Utilities|String|StringToName (Utilities|String|Append "p" (Utilities|String|ToString(Integer) _array_index) "a")) (Utilities|Struct|MakeLinearColor (.x (Utilities|Map|Find _socketlocationmap_21_abf6aa244a5f84728a5e83be2328c7fa _array_element)) (.y (Utilities|Map|Find _socketlocationmap_21_abf6aa244a5f84728a5e83be2328c7fa _array_element)) (.z (Utilities|Map|Find _socketlocationmap_21_abf6aa244a5f84728a5e83be2328c7fa _array_element)) (* (Math|Vector|VectorLength (Utilities|Map|Find _socketvelocitymap_26_82b0e24b45935a12e1949f918a59a537 _array_element)) (* (Math|Float|Sign(Float) (+ (.z (Utilities|Map|Find _socketvelocitymap_26_82b0e24b45935a12e1949f918a59a537 _array_element)) 0.000135)) _forcestrength_33_2caa30794d1eff60ae1c3491d011cecf))) :!class /Script/Engine.MaterialInstanceDynamic)
            (Rendering|Material|SetVectorParameterValue (LocalVariables|GetLocalMID) (Utilities|String|StringToName (Utilities|String|Append "p" (Utilities|String|ToString(Integer) _array_index) "b")) (Utilities|Struct|MakeLinearColor (.x (Utilities|Map|Find _socketlocationmap_21_abf6aa244a5f84728a5e83be2328c7fa (Utilities|Map|Find _socketsandendpoints_6_b3edd8fc43a7c681151f46be0aa158c5 _array_element))) (.y (Utilities|Map|Find _socketlocationmap_21_abf6aa244a5f84728a5e83be2328c7fa (Utilities|Map|Find _socketsandendpoints_6_b3edd8fc43a7c681151f46be0aa158c5 _array_element))) (.z (Utilities|Map|Find _socketlocationmap_21_abf6aa244a5f84728a5e83be2328c7fa (Utilities|Map|Find _socketsandendpoints_6_b3edd8fc43a7c681151f46be0aa158c5 _array_element))) (* (Math|Vector|VectorLength (Utilities|Map|Find _socketvelocitymap_26_82b0e24b45935a12e1949f918a59a537 (Utilities|Map|Find _socketsandendpoints_6_b3edd8fc43a7c681151f46be0aa158c5 _array_element))) (* (Math|Float|Sign(Float) (+ (.z (Utilities|Map|Find _socketvelocitymap_26_82b0e24b45935a12e1949f918a59a537 (Utilities|Map|Find _socketsandendpoints_6_b3edd8fc43a7c681151f46be0aa158c5 _array_element))) 0.000135)) _forcestrength_33_2caa30794d1eff60ae1c3491d011cecf))) :!class /Script/Engine.MaterialInstanceDynamic)))
        (Rendering|Material|SetScalarParameterValue _localmid_2 "CapsuleWidth" (* (Utilities|Struct|BreakFluidForceDynamic (Utilities|Struct|BreakFluidForceDynamicPerInstanceData (LocalVariables|GetLocalForce))) (Utilities|Struct|BreakFluidForceDynamic (Utilities|Struct|BreakFluidForceDynamicPerInstanceData (LocalVariables|GetLocalForce)))) :!class /Script/Engine.MaterialInstanceDynamic)
        (Rendering|Material|SetVectorParameterValue _localmid_2 "GridLocation" (Math|Conversions|ToLinearColor(Vector) (Variables|Debug|GetGridCenter)) :!class /Script/Engine.MaterialInstanceDynamic)
        (Rendering|Material|SetScalarParameterValue _localmid_2 "WaterZ" _waterlevel_34_a4e505d148073b883ca7b1b09a3e34a8_2 :!class /Script/Engine.MaterialInstanceDynamic)
        (Canvas|DrawMaterial Canvas (LocalVariables|GetLocalMID) _screen_position_2 _screen_size_2 (/ _screen_position_2 (Math|Conversions|ToFloat(Integer) (Variables|Simulation|GetResolution))) (/ _screen_size_2 (Math|Conversions|ToFloat(Integer) (Variables|Simulation|GetResolution))) 0.0 "(X=0.500000,Y=0.500000)")))))

(fn DrawImpulseForce
  (params (Canvas :type Canvas%20Object%20Reference) (Canvas%20Size :type Vector%202D%20Structure) (Impulse%20Settings :type Fluid%20Force%20Impulse%20Per%20Instance%20Data%20Structure))
  (bind (_impulsesettings_41_c9a94c02422d8bf40df6b1bb2a0d8cbc _elapsedtime_44_12c387c1450456e47fc74bbd11eeae4f _startoffset_48_78ef0e3f4b05f70c5679f9b9f41d590f _issplasheffect_54_4a5b178940d26e71d6fcdf84584a5284) (Utilities|Struct|BreakFluidForceImpulsePerInstanceData Impulse%20Settings))
  (bind (_worldposition_32_c9a94c02422d8bf40df6b1bb2a0d8cbc _forceradius_31_c31b527c4c367a5ca5e1df8e49e76234 _forcestrength_30_2caa30794d1eff60ae1c3491d011cecf _materialoverride_24_5a792ce8489a59e5d9b24f9e4dcfe94a) (Utilities|Struct|BreakFluidForceImpulse _impulsesettings_41_c9a94c02422d8bf40df6b1bb2a0d8cbc))
  (bind (_impulsesettings_41_c9a94c02422d8bf40df6b1bb2a0d8cbc_2 _elapsedtime_44_12c387c1450456e47fc74bbd11eeae4f_2 _startoffset_48_78ef0e3f4b05f70c5679f9b9f41d590f_2) (Utilities|Struct|BreakFluidForceImpulsePerInstanceData Impulse%20Settings))
  (bind (_forcestrength_30_2caa30794d1eff60ae1c3491d011cecf_2 _materialoverride_24_5a792ce8489a59e5d9b24f9e4dcfe94a_2 _lifetime_34_c2749c1449c41d4f236bcbaf6ed34113 _strengthoverlifepower_41_85b52c994a7ed323a34bf4bbcb0dfa0f _timedwaterdropsplashes_38_2cb1456b483aef3a52677aaf4152e31e) (Utilities|Struct|BreakFluidForceImpulse _impulsesettings_41_c9a94c02422d8bf40df6b1bb2a0d8cbc_2))
  (bind (_screen_position _screen_size) (Class|RendertoTextureFunctionLibrary|SetCanvasMaterialScaleAndPosition Canvas%20Size (Math|Conversions|ToVector2D(Vector) (CallFunction|ConvertForcePosition _worldposition_32_c9a94c02422d8bf40df6b1bb2a0d8cbc)) (/ _forceradius_31_c31b527c4c367a5ca5e1df8e49e76234 (* (Variables|Simulation|GetFluid%20Size) 0.5))))
  (bind _returnvalue (- _elapsedtime_44_12c387c1450456e47fc74bbd11eeae4f_2 _startoffset_48_78ef0e3f4b05f70c5679f9b9f41d590f_2))
  (if _issplasheffect_54_4a5b178940d26e71d6fcdf84584a5284
    (if (>= _returnvalue 0.0)
      (bind _mid (Default|GetForceMID (select (Utilities|IsValid _materialoverride_24_5a792ce8489a59e5d9b24f9e4dcfe94a_2 :!class /Script/Engine.KismetSystemLibrary) _materialoverride_24_5a792ce8489a59e5d9b24f9e4dcfe94a_2 (Variables|Forces|GetDefaultSplashMaterial))))
      (bind _output_get (LocalVariables|SetLocalImpulseMID _mid))
      (Rendering|Material|SetScalarParameterValue _output_get "F" (* _forcestrength_30_2caa30794d1eff60ae1c3491d011cecf_2 (select (and (> (Math|Float|Clamp(Float) (/ _returnvalue _lifetime_34_c2749c1449c41d4f236bcbaf6ed34113)) 0) (> _strengthoverlifepower_41_85b52c994a7ed323a34bf4bbcb0dfa0f 0)) (Math|Float|Power (Math|Float|Clamp(Float) (/ _returnvalue _lifetime_34_c2749c1449c41d4f236bcbaf6ed34113)) _strengthoverlifepower_41_85b52c994a7ed323a34bf4bbcb0dfa0f) 1.0)) :!class /Script/Engine.MaterialInstanceDynamic)
      (Rendering|Material|SetScalarParameterValue _output_get "DropDensity" (Utilities|Struct|BreakFluidForceImpulseTimedEffects _timedwaterdropsplashes_38_2cb1456b483aef3a52677aaf4152e31e) :!class /Script/Engine.MaterialInstanceDynamic))
    (else
      (bind _mid_2 (Default|GetForceMID _materialoverride_24_5a792ce8489a59e5d9b24f9e4dcfe94a))
      (bind _output_get_2 (LocalVariables|SetLocalImpulseMID _mid_2))
      (Rendering|Material|SetScalarParameterValue (LocalVariables|GetLocalImpulseMID) "F" _forcestrength_30_2caa30794d1eff60ae1c3491d011cecf :!class /Script/Engine.MaterialInstanceDynamic)))
  (Canvas|DrawMaterial :self Canvas :RenderMaterial (LocalVariables|GetLocalImpulseMID) :ScreenPosition _screen_position :ScreenSize _screen_size :CoordinateSize "(X=1.000000,Y=1.000000)" :Rotation 0.0 :PivotPoint "(X=0.500000,Y=0.500000)"))

(fn GetForceMID
  (params (Parent :type Material%20Interface%20Object%20Reference%20%28by%20ref%29))
  (bind _forceparentandmidmap (Variables|Forces|GetForceParentAndMIDMap))
  (bind (_value _returnvalue) (Utilities|Map|Find _forceparentandmidmap Parent))
  (LocalVariables|SetOutMID 0)
  (if _returnvalue
    (Utilities|IsValid _value
      (:Is%20Valid
        (LocalVariables|SetOutMID _value))
      (:Is%20Not%20Valid
        (bind _returnvalue_2 (Rendering|Material|CreateDynamicMaterialInstance Parent (Utilities|String|StringToName (Utilities|String|Append "FluidForceMID_" (Utilities|GetDisplayName Parent))) "None" :!class /Script/Engine.KismetMaterialLibrary :!label _label))
        (bind _output_get (LocalVariables|SetOutMID _returnvalue_2))
        (Rendering|Material|SetScalarParameterValue _output_get "FluidSize" (Variables|Simulation|GetFluid%20Size) :!class /Script/Engine.MaterialInstanceDynamic)
        (Rendering|Material|SetTextureParameterValue _output_get "WaterVelocityAndHeight" (Variables|Simulation|GetWaterVelocityTexture))
        (Rendering|Material|SetScalarParameterValue _output_get "WaterZoneIndex" (Math|Conversions|ToFloat(Integer) (Variables|Debug|GetWaterZoneIndex)) :!class /Script/Engine.MaterialInstanceDynamic)
        (Utilities|Map|Add (Variables|Forces|GetForceParentAndMIDMap) Parent (LocalVariables|GetOutMID))))
    (return (LocalVariables|GetOutMID)))
  (goto _label))

(fn GetFramesSinceLastActiveForces
  (bind _keys (Utilities|Map|Keys (Variables|Forces|GetDynamicForces)))
  (bind _keys_2 (Utilities|Map|Keys (Variables|Forces|GetProjectileForces)))
  (if (or (or (> (Utilities|Array|Length _keys_2) 0) (> (Utilities|Array|Length _keys) 0)) (> (Utilities|Array|Length (Variables|Forces|GetImpulseForces)) 0))
    (Variables|Default|SetFrames%20Since%20Last%20Active%20Force 0)
    (else
      (Math|Integer|IncrementInt (Variables|Default|GetFrames%20Since%20Last%20Active%20Force)))))

(fn GetLandscapeWaterInfo
  (params (WaterZone :type Water%20Zone%20Object%20Reference))
  (bind _outactors (Actor|GetAllActorsOfClass "/Script/Water.WaterZone"))
  (bind _array_element (Utilities|Array|ForEachLoopwithBreak _outactors :!label _label)
    (:LoopBody
      (Utilities|IsValid (Variables|/Script/Water.WaterZone|GetWaterInfoTextureArray _array_element)
        (:Is%20Valid
          (if (== (Utilities|Struct|BreakIntPoint (Variables|/Script/Water.WaterZone|GetRenderTargetResolution _array_element)) 768)
            (Variables|Simulation|SetWaterVelocityTexture (Variables|/Script/Water.WaterZone|GetWaterInfoTextureArray _array_element))
            (goto _label Break)))))
    (:Completed
      (return))))

(fn GetLocalPawn
  (return (Pawn|GetControlledPawn (Game|GetPlayerController 0))))

(fn GetNormalRT
  (return (Variables|Debug|GetNormalRT)))

(fn GetPlayerPawnForces)

(fn GridMovement
  (bind _prevloc (Variables|Default|GetPrevLoc))
  (bind _returnvalue (Math|Conversions|ToLinearColor(Vector) _prevloc))
  (bind _x (.x _prevloc))
  (bind _y (.y _prevloc))
  (bind _ripplesimmid (Variables|Debug|GetRippleSimMID))
  (Variables|Debug|SetGridCenter (Utilities|GetGridCenter))
  (Utilities|FlowControl|Sequence
    (:then_0
      (Variables|Default|SetPrevOffset (/ (- (Variables|Default|GetPrevLoc) (Variables|Debug|GetGridCenter)) (Variables|Simulation|GetFluid%20Size)))
      (Variables|Default|SetPrevOffset2 (/ (- (Variables|Default|GetPrecLoc2) (Variables|Debug|GetGridCenter)) (Variables|Simulation|GetFluid%20Size)))
      (Rendering|Material|SetVectorParameterValue _ripplesimmid "PreviousOffsets1" (Math|Conversions|ToLinearColor(Vector) (Variables|Default|GetPrevOffset)) :!class /Script/Engine.MaterialInstanceDynamic)
      (Rendering|Material|SetVectorParameterValue _ripplesimmid "PreviousOffsets2" (Math|Conversions|ToLinearColor(Vector) (Variables|Default|GetPrevOffset2)) :!class /Script/Engine.MaterialInstanceDynamic)
      (Rendering|Material|SetVectorParameterValue _ripplesimmid "GridLocation" _returnvalue :!class /Script/Engine.MaterialInstanceDynamic)
      (Rendering|Material|SetVectorParameterValue "/Water/Materials/MPC/MPC_Water.MPC_Water" "SimLocation" _returnvalue :!class /Script/Engine.KismetMaterialLibrary)
      (Utilities|FlowControl|Sequence
        (:then_0
          (if (Variables|Simulation|GetShow%20Simulation%20Mesh)
            (Transformation|SetWorldLocation (Variables|Default|GetFluid%20Display%20Mesh) (Math|Vector|MakeVector _x _y (.z (Transformation|GetActorLocation))) false false)))
        (:then_1
          (if (Variables|Debug|GetShowDebugWaterPlane)
            (Transformation|SetWorldLocation (Variables|Default|GetDebugPlane) (Math|Vector|MakeVector _x _y (+ (.z _prevloc) 75.0)) false false)))))
    (:then_1
      (Variables|Default|SetPrecLoc2 (Variables|Default|GetPrevLoc))
      (Variables|Default|SetPrevLoc (Variables|Debug|GetGridCenter)))))

(fn RegisterDynamicForce
  (params (Dynamic%20Fluid%20Force :type Fluid%20Force%20Dynamic%20Structure) (Tracked%20Component :type Scene%20Component%20Object%20Reference) (WaterLevel :type Float%20%28double-precision%29))
  (bind _localdynamicforce (LocalVariables|GetLocalDynamicForce))
  (bind (_forcecomponent_34_abf6640f40d37677ef6f809e09046055 _skeletalmeshsetup_31_51a4130440baffba1da0fe83e942a30a) (Utilities|Struct|BreakFluidForceDynamic _localdynamicforce))
  (bind _socketsandendpoints_6_b3edd8fc43a7c681151f46be0aa158c5 (Utilities|Struct|BreakFluidForceSocketInfo _skeletalmeshsetup_31_51a4130440baffba1da0fe83e942a30a))
  (bind (_forceradius_32_c31b527c4c367a5ca5e1df8e49e76234 _forcecomponent_34_abf6640f40d37677ef6f809e09046055_2) (Utilities|Struct|BreakFluidForceDynamic (LocalVariables|GetLocalDynamicForce)))
  (bind (_origin _boxextent _sphereradius) (Collision|GetComponentBounds (Utilities|Struct|BreakFluidForceDynamic (LocalVariables|GetLocalDynamicForce))))
  (Networking|CanExecuteCosmeticEvents
    (:True
      (bind _output_get (LocalVariables|SetLocalDynamicForce Dynamic%20Fluid%20Force))
      (Utilities|Struct|SetmembersinFluidForceDynamic _output_get Tracked%20Component)
      (LocalVariables|SetLocalWaterLevel WaterLevel)
      (bind _returnvalue (Water|GetShallowWaterMaxDynamicForces))
      (if (< (Utilities|Map|Length (Variables|Forces|GetDynamicForces)) _returnvalue)
        (Utilities|Map|Clear (LocalVariables|GetSocketLocMap))
        (Utilities|Map|Clear (LocalVariables|GetSocketVelMap))
        (switch Utilities|FlowControl|Switch|SwitchonFluidDynamicForceMeshType (Utilities|Struct|BreakFluidForceDynamic (LocalVariables|GetLocalDynamicForce))
          (:NewEnumerator0
            ;[1 1 1] Add Simple Component Location Force
            (Utilities|Struct|SetmembersinFluidForceDynamic (LocalVariables|GetLocalDynamicForce) (select (Utilities|IsValid (Utilities|Struct|BreakFluidForceDynamic (LocalVariables|GetLocalDynamicForce)) :!class /Script/Engine.KismetSystemLibrary) (Utilities|Struct|BreakFluidForceDynamic (LocalVariables|GetLocalDynamicForce)) "/FortWater/FluidSimulation/Materials/Forces/M_Fluid_Sim_Force_Component.M_Fluid_Sim_Force_Component"))
            (Utilities|Map|Add (Variables|Forces|GetDynamicForces) _forcecomponent_34_abf6640f40d37677ef6f809e09046055_2 (Utilities|Struct|MakeFluidForceDynamicPerInstanceData (LocalVariables|GetLocalDynamicForce) (Transformation|GetWorldLocation _forcecomponent_34_abf6640f40d37677ef6f809e09046055_2) "0.000000,0.000000,0.000000" _forceradius_32_c31b527c4c367a5ca5e1df8e49e76234 (LocalVariables|GetLocalWaterLevel) (Utilities|Map|MakeMap "None" "0, 0, 0") (Utilities|Map|MakeMap "None" "0, 0, 0"))))
          (:NewEnumerator1
            (Utilities|Struct|SetmembersinFluidForceDynamic (LocalVariables|GetLocalDynamicForce) (select (Utilities|IsValid (Utilities|Struct|BreakFluidForceDynamic (LocalVariables|GetLocalDynamicForce)) :!class /Script/Engine.KismetSystemLibrary) (Utilities|Struct|BreakFluidForceDynamic (LocalVariables|GetLocalDynamicForce)) "/FortWater/FluidSimulation/Materials/Forces/M_Fluid_Force_Skelmesh.M_Fluid_Force_Skelmesh"))
            (Utilities|Casting|CastToSkeletalMeshComponent _forcecomponent_34_abf6640f40d37677ef6f809e09046055)
            ;[0.9613 0.3245 1] Setup Position and Velocity Maps
            (bind _keys (Utilities|Map|Keys _socketsandendpoints_6_b3edd8fc43a7c681151f46be0aa158c5))
            (for _array_element _keys
              (Utilities|Map|Add (LocalVariables|GetSocketLocMap) _array_element)
              (Utilities|Map|Add (LocalVariables|GetSocketVelMap) _array_element))
            (bind _values (Utilities|Map|Values _socketsandendpoints_6_b3edd8fc43a7c681151f46be0aa158c5))
            (for _array_element_2 _values
              (Utilities|Map|Add (LocalVariables|GetSocketLocMap) _array_element_2)
              (Utilities|Map|Add (LocalVariables|GetSocketVelMap) _array_element_2))
            ;[1 0.5316 0.09549] Set Initial Position Values
            (bind _keys_2 (Utilities|Map|Keys (LocalVariables|GetSocketLocMap)))
            (for _array_element_3 _keys_2
              (Utilities|Map|Add (LocalVariables|GetSocketLocMap) _array_element_3 (Transformation|GetSocketLocation (Utilities|Struct|BreakFluidForceDynamic (LocalVariables|GetLocalDynamicForce)) _array_element_3)))
            ;[1 1 1] Add Skeletal Mesh Force
            (Utilities|Map|Add (Variables|Forces|GetDynamicForces) (Utilities|Struct|BreakFluidForceDynamic (LocalVariables|GetLocalDynamicForce)) (Utilities|Struct|MakeFluidForceDynamicPerInstanceData (LocalVariables|GetLocalDynamicForce) (Transformation|GetWorldLocation (Utilities|Struct|BreakFluidForceDynamic (LocalVariables|GetLocalDynamicForce))) "0.000000,0.000000,0.000000" _sphereradius (LocalVariables|GetLocalWaterLevel) (LocalVariables|GetSocketLocMap) (LocalVariables|GetSocketVelMap)))))))))

(fn RegisterProjectileForce
  (params (Dynamic%20Fluid%20Force :type Fluid%20Force%20Dynamic%20Structure) (Tracked%20Component :type Scene%20Component%20Object%20Reference))
  (bind _localdynamicforce (LocalVariables|GetLocalDynamicForce))
  (bind (_forcecomponent_34_abf6640f40d37677ef6f809e09046055 _skeletalmeshsetup_31_51a4130440baffba1da0fe83e942a30a) (Utilities|Struct|BreakFluidForceDynamic _localdynamicforce))
  (bind _socketsandendpoints_6_b3edd8fc43a7c681151f46be0aa158c5 (Utilities|Struct|BreakFluidForceSocketInfo _skeletalmeshsetup_31_51a4130440baffba1da0fe83e942a30a))
  (bind (_forceradius_32_c31b527c4c367a5ca5e1df8e49e76234 _forcecomponent_34_abf6640f40d37677ef6f809e09046055_2) (Utilities|Struct|BreakFluidForceDynamic (LocalVariables|GetLocalDynamicForce)))
  (bind (_origin _boxextent _sphereradius) (Collision|GetComponentBounds (Utilities|Struct|BreakFluidForceDynamic (LocalVariables|GetLocalDynamicForce))))
  (Networking|CanExecuteCosmeticEvents
    (:True
      (bind _output_get (LocalVariables|SetLocalDynamicForce Dynamic%20Fluid%20Force))
      (Utilities|Struct|SetmembersinFluidForceDynamic _output_get Tracked%20Component)
      (bind _returnvalue (Water|GetShallowWaterMaxDynamicForces))
      (if (< (Utilities|Map|Length (Variables|Forces|GetDynamicForces)) _returnvalue)
        (Utilities|Map|Clear (LocalVariables|GetSocketLocMap))
        (Utilities|Map|Clear (LocalVariables|GetSocketVelMap))
        (switch Utilities|FlowControl|Switch|SwitchonFluidDynamicForceMeshType (Utilities|Struct|BreakFluidForceDynamic (LocalVariables|GetLocalDynamicForce))
          (:NewEnumerator0
            ;[1 1 1] Add Simple Component Location Force
            (Utilities|Struct|SetmembersinFluidForceDynamic (LocalVariables|GetLocalDynamicForce) (select (Utilities|IsValid (Utilities|Struct|BreakFluidForceDynamic (LocalVariables|GetLocalDynamicForce)) :!class /Script/Engine.KismetSystemLibrary) (Utilities|Struct|BreakFluidForceDynamic (LocalVariables|GetLocalDynamicForce)) "/FortWater/FluidSimulation/Materials/Forces/M_Fluid_Sim_Force_Component.M_Fluid_Sim_Force_Component"))
            (Utilities|Map|Add (Variables|Forces|GetProjectileForces) _forcecomponent_34_abf6640f40d37677ef6f809e09046055_2 (Utilities|Struct|MakeFluidForceDynamicPerInstanceData (LocalVariables|GetLocalDynamicForce) (Transformation|GetWorldLocation _forcecomponent_34_abf6640f40d37677ef6f809e09046055_2) "0.000000,0.000000,0.000000" _forceradius_32_c31b527c4c367a5ca5e1df8e49e76234 0.0 (Utilities|Map|MakeMap "None" "0, 0, 0") (Utilities|Map|MakeMap "None" "0, 0, 0"))))
          (:NewEnumerator1
            (Utilities|Struct|SetmembersinFluidForceDynamic (LocalVariables|GetLocalDynamicForce) (select (Utilities|IsValid (Utilities|Struct|BreakFluidForceDynamic (LocalVariables|GetLocalDynamicForce)) :!class /Script/Engine.KismetSystemLibrary) (Utilities|Struct|BreakFluidForceDynamic (LocalVariables|GetLocalDynamicForce)) "/FortWater/FluidSimulation/Materials/Forces/M_Fluid_Force_Skelmesh.M_Fluid_Force_Skelmesh"))
            (Utilities|Casting|CastToSkeletalMeshComponent _forcecomponent_34_abf6640f40d37677ef6f809e09046055)
            ;[0.9613 0.3245 1] Setup Position and Velocity Maps
            (bind _keys (Utilities|Map|Keys _socketsandendpoints_6_b3edd8fc43a7c681151f46be0aa158c5))
            (for _array_element _keys
              (Utilities|Map|Add (LocalVariables|GetSocketLocMap) _array_element)
              (Utilities|Map|Add (LocalVariables|GetSocketVelMap) _array_element))
            (bind _values (Utilities|Map|Values _socketsandendpoints_6_b3edd8fc43a7c681151f46be0aa158c5))
            (for _array_element_2 _values
              (Utilities|Map|Add (LocalVariables|GetSocketLocMap) _array_element_2)
              (Utilities|Map|Add (LocalVariables|GetSocketVelMap) _array_element_2))
            ;[1 0.5316 0.09549] Set Initial Position Values
            (bind _keys_2 (Utilities|Map|Keys (LocalVariables|GetSocketLocMap)))
            (for _array_element_3 _keys_2
              (Utilities|Map|Add (LocalVariables|GetSocketLocMap) _array_element_3 (Transformation|GetSocketLocation (Utilities|Struct|BreakFluidForceDynamic (LocalVariables|GetLocalDynamicForce)) _array_element_3)))
            ;[1 1 1] Add Skeletal Mesh Force
            (Utilities|Map|Add (Variables|Forces|GetProjectileForces) (Utilities|Struct|BreakFluidForceDynamic (LocalVariables|GetLocalDynamicForce)) (Utilities|Struct|MakeFluidForceDynamicPerInstanceData (LocalVariables|GetLocalDynamicForce) (Transformation|GetWorldLocation (Utilities|Struct|BreakFluidForceDynamic (LocalVariables|GetLocalDynamicForce))) "0.000000,0.000000,0.000000" _sphereradius 0.0 (LocalVariables|GetSocketLocMap) (LocalVariables|GetSocketVelMap)))))))))

(fn RemoveDynamicForce
  (params (Component :type Actor%20Component%20Object%20Reference))
  (bind (_value _returnvalue) (Utilities|Map|Find (Variables|Forces|GetDynamicForces) Component))
  (if _returnvalue
    (Utilities|Map|Remove (Variables|Forces|GetDynamicForces) Component)))

(fn RemoveProjectileForce
  (params (Component :type Actor%20Component%20Object%20Reference))
  (bind (_value _returnvalue) (Utilities|Map|Find (Variables|Forces|GetProjectileForces) Component))
  (if _returnvalue
    (Utilities|Map|Remove (Variables|Forces|GetProjectileForces) Component)))

(fn SetWaterbodyMIDParams
  (bind _outactors (Actor|GetAllActorsOfClass "/Script/Water.WaterBody"))
  (for _array_element _outactors
    (bind _returnvalue (Rendering|GetWaterMaterialInstance (Water|GetWaterBodyComponent _array_element)))
    (Rendering|Material|SetTextureParameterValue _returnvalue "NormalAndHeight" (Variables|Debug|GetNormalRT))
    (Rendering|Material|SetTextureParameterValue _returnvalue "Forces" (Variables|Debug|GetForcesRT))
    (bind _returnvalue_2 (Rendering|GetRivertoOceanTransitionMaterialInstance (Water|GetWaterBodyComponent _array_element)))
    (Utilities|IsValid _returnvalue_2
      (:Is%20Valid
        (Rendering|Material|SetTextureParameterValue _returnvalue_2 "NormalAndHeight" (Variables|Debug|GetNormalRT)))
      (:Is%20Not%20Valid
        _))
    (bind _returnvalue_3 (Rendering|GetRivertoLakeTransitionMaterialInstance (Water|GetWaterBodyComponent _array_element)))
    (Utilities|IsValid _returnvalue_3
      (:Is%20Valid
        (Rendering|Material|SetTextureParameterValue _returnvalue_3 "NormalAndHeight" (Variables|Debug|GetNormalRT))))))

(fn SetSingleWaterBodyMIDTexture
  (params (Waterbody :type Water%20Body%20Object%20Reference))
  (bind _waterbodycomponent (Variables|/Script/Water.WaterBody|GetWaterBodyComponent Waterbody))
  (bind _returnvalue (Rendering|GetWaterMaterialInstance _waterbodycomponent))
  (Rendering|Material|SetTextureParameterValue _returnvalue "NormalAndHeight" (Variables|Debug|GetNormalRT))
  (bind _returnvalue_2 (Rendering|GetRivertoOceanTransitionMaterialInstance _waterbodycomponent))
  (Utilities|IsValid _returnvalue_2
    (:Is%20Valid
      (Rendering|Material|SetTextureParameterValue _returnvalue_2 "NormalAndHeight" (Variables|Debug|GetNormalRT)))
    (:Is%20Not%20Valid
      _))
  (bind _returnvalue_3 (Rendering|GetRivertoLakeTransitionMaterialInstance _waterbodycomponent))
  (Utilities|IsValid _returnvalue_3
    (:Is%20Valid
      (Rendering|Material|SetTextureParameterValue _returnvalue_3 "NormalAndHeight" (Variables|Debug|GetNormalRT)))))

(fn SetupDisplayMIDs
  (if (Variables|Simulation|GetShow%20Simulation%20Mesh)
    (bind _returnvalue (Rendering|Material|CreateDynamicMaterialInstance (Variables|Simulation|GetDisplay%20Material) "None" "None" :!class /Script/Engine.KismetMaterialLibrary))
    (Variables|Debug|SetDisplayMID _returnvalue)
    (Rendering|Material|SetMaterial (Variables|Default|GetFluid%20Display%20Mesh) 0 (Variables|Debug|GetDisplayMID))
    (bind _returnvalue_2 (Rendering|Material|CreateDynamicMaterialInstance "/FortWater/FluidSimulation/Materials/M_Underwater_CrossSection_Test.M_Underwater_CrossSection_Test" "None" "None" :!class /Script/Engine.KismetMaterialLibrary))
    (Variables|Debug|SetCrossSectionMID _returnvalue_2)
    (Rendering|Material|SetTextureParameterValue (Variables|Debug|GetDisplayMID) "NormalAndHeight" (Variables|Debug|GetNormalRT))
    (Rendering|Material|SetScalarParameterValue (Variables|Debug|GetDisplayMID) "Cut" (select (Variables|Debug|GetShow%20Cross%20Section) (* (+ (/ (.x (Variables|Debug|GetCutPos)) (* (Variables|Simulation|GetFluid%20Size) 0.5)) 1.0) 0.5) 0.0) :!class /Script/Engine.MaterialInstanceDynamic)
    (Rendering|Material|SetScalarParameterValue (Variables|Debug|GetDisplayMID) "WaterZoneIndex" (Math|Conversions|ToFloat(Integer) (Variables|Debug|GetWaterZoneIndex)) :!class /Script/Engine.MaterialInstanceDynamic)
    (if (Variables|Debug|GetShow%20Cross%20Section)
      (Rendering|Material|SetMaterial (Variables|Default|GetCross%20Section%20Mesh) 0 (Variables|Debug|GetCrossSectionMID)))))

(fn SetupSimMIDs
  (bind (_x _y) (Math|Vector2D|BreakVector2D (Variables|Debug|GetWaterZoneExtent)))
  (bind _rendernormalsmid (Variables|Debug|GetRenderNormalsMID))
  (bind _fluid_size (Variables|Simulation|GetFluid%20Size))
  (bind _ripplesimmid (Variables|Debug|GetRippleSimMID))
  (Landmass|CreateAndAssignMID (Variables|Debug|GetRippleSimMID) (select (Variables|Simulation|GetUse%20Terrain%20Water%20System) "/FortWater/FluidSimulation/Materials/Simulation/M_Fluid_Sim_01.M_Fluid_Sim_01" "/FortWater/FluidSimulation/Materials/Simulation/M_Fluid_Sim_01_NoTerrain.M_Fluid_Sim_01_NoTerrain")
    (:Out
      (Landmass|CreateAndAssignMID (Variables|Debug|GetPhysicsForceMID) "/FortWater/FluidSimulation/Materials/Forces/PhysicsFluidForce_RT.PhysicsFluidForce_RT"
        (:Out
          (Rendering|Material|SetScalarParameterValue _ripplesimmid "Damping" (Variables|Simulation|GetDamping) :!class /Script/Engine.MaterialInstanceDynamic)
          (Rendering|Material|SetScalarParameterValue _ripplesimmid "TravelSpeed" (Variables|Simulation|GetTravel%20Speed) :!class /Script/Engine.MaterialInstanceDynamic)
          (Rendering|Material|SetScalarParameterValue _ripplesimmid "Resolution" (Math|Conversions|ToFloat(Integer) (Variables|Simulation|GetResolution)) :!class /Script/Engine.MaterialInstanceDynamic)
          (Rendering|Material|SetScalarParameterValue _ripplesimmid "FluidSize" (Variables|Simulation|GetFluid%20Size) :!class /Script/Engine.MaterialInstanceDynamic)
          (Rendering|Material|SetTextureParameterValue _ripplesimmid "Forces" (Variables|Debug|GetForcesRT))
          (Rendering|Material|SetScalarParameterValue _ripplesimmid "TimeStep" (Variables|Default|GetFixedStep) :!class /Script/Engine.MaterialInstanceDynamic)
          (Rendering|Material|SetTextureParameterValue _ripplesimmid "NormalAndHeight" (Variables|Debug|GetNormalRT))
          (Utilities|IsValid (Variables|Debug|GetRenderNormalsMID)
            (:Is%20Valid
              _)
            (:Is%20Not%20Valid
              (bind _returnvalue (Rendering|Material|CreateDynamicMaterialInstance "/FortWater/FluidSimulation/Materials/Simulation/M_RT_Normals.M_RT_Normals" "None" "None" :!class /Script/Engine.KismetMaterialLibrary))
              (Variables|Debug|SetRenderNormalsMID _returnvalue)))
          (Rendering|Material|SetScalarParameterValue _rendernormalsmid "Resolution" (Math|Conversions|ToFloat(Integer) (Variables|Simulation|GetResolution)) :!class /Script/Engine.MaterialInstanceDynamic)
          (Rendering|Material|SetScalarParameterValue _rendernormalsmid "FluidSize" _fluid_size :!class /Script/Engine.MaterialInstanceDynamic)
          (Rendering|Material|SetScalarParameterValue "/Water/Materials/MPC/MPC_Water.MPC_Water" "FluidSimSize" _fluid_size :!class /Script/Engine.KismetMaterialLibrary)
          (Rendering|Material|SetTextureParameterValue (Variables|Debug|GetRippleSimMID) "WaterVelocityAndHeight" (Variables|Simulation|GetWaterVelocityTexture))
          (Rendering|Material|SetScalarParameterValue (Variables|Debug|GetRippleSimMID) "WaterZoneIndex" (Math|Conversions|ToFloat(Integer) (Variables|Debug|GetWaterZoneIndex)) :!class /Script/Engine.MaterialInstanceDynamic)
          (Rendering|Material|SetScalarParameterValue (Variables|Debug|GetRippleSimMID) "MaxFlowVelocity" (Variables|Debug|GetMaxFlowVelocity) :!class /Script/Engine.MaterialInstanceDynamic)
          (Rendering|Material|SetVectorParameterValue (Variables|Debug|GetRippleSimMID) "WaterZoneLocation" (Math|Conversions|ToLinearColor(Vector) (Variables|Debug|GetWaterZoneLocation)) :!class /Script/Engine.MaterialInstanceDynamic)
          (Rendering|Material|SetVectorParameterValue (Variables|Debug|GetRippleSimMID) "WaterZoneExtent" (Utilities|Struct|MakeLinearColor _x _y) :!class /Script/Engine.MaterialInstanceDynamic)
          ;[1 1 1] Setup Initial Default Force Material Map
          (bind _keys (Utilities|Map|Keys (Variables|Forces|GetForceParentAndMIDMap)))
          (for _array_element _keys
            (bind _mid (Default|GetForceMID _array_element))
            (Rendering|Material|SetScalarParameterValue _mid "LandscapeWaterSystem" (Math|Conversions|ToFloat(Boolean) (Variables|Simulation|GetUse%20Terrain%20Water%20System)) :!class /Script/Engine.MaterialInstanceDynamic)))))))

(fn UpdateDynamicForces
  (bind _localforce (LocalVariables|GetLocalForce))
  (bind (_forceinfo_2_a6a35e3243faf59d161a5fbaa2f6f2b1 _componentlocation_13_959307184c8e5ccaca55fc8378d92cfd _boundsradius_30_adff818743be39ac4a481d995cb50d03 _waterlevel_34_a4e505d148073b883ca7b1b09a3e34a8) (Utilities|Struct|BreakFluidForceDynamicPerInstanceData _localforce))
  (bind _forcecomponent_34_abf6640f40d37677ef6f809e09046055 (Utilities|Struct|BreakFluidForceDynamic _forceinfo_2_a6a35e3243faf59d161a5fbaa2f6f2b1))
  (bind _localforce_2 (LocalVariables|GetLocalForce))
  (bind (_boundsradius_30_adff818743be39ac4a481d995cb50d03_2 _waterlevel_34_a4e505d148073b883ca7b1b09a3e34a8_2 _socketlocationmap_21_abf6aa244a5f84728a5e83be2328c7fa _socketvelocitymap_26_82b0e24b45935a12e1949f918a59a537) (Utilities|Struct|BreakFluidForceDynamicPerInstanceData _localforce_2))
  (bind (_forcetype_28_ddc16ee543d2dfd3ba29c49d32198c9c _forcecomponent_34_abf6640f40d37677ef6f809e09046055_2) (Utilities|Struct|BreakFluidForceDynamic (Utilities|Struct|BreakFluidForceDynamicPerInstanceData (LocalVariables|GetLocalForce))))
  (bind _localforce_3 (LocalVariables|GetLocalForce))
  (bind (_boundsradius_30_adff818743be39ac4a481d995cb50d03_3 _waterlevel_34_a4e505d148073b883ca7b1b09a3e34a8_3 _socketlocationmap_21_abf6aa244a5f84728a5e83be2328c7fa_2 _socketvelocitymap_26_82b0e24b45935a12e1949f918a59a537_2) (Utilities|Struct|BreakFluidForceDynamicPerInstanceData _localforce_3))
  (bind _localforce_4 (LocalVariables|GetLocalForce))
  (bind _forceinfo_2_a6a35e3243faf59d161a5fbaa2f6f2b1_2 (Utilities|Struct|BreakFluidForceDynamicPerInstanceData _localforce_4))
  (bind (_forcetype_28_ddc16ee543d2dfd3ba29c49d32198c9c_2 _forcecomponent_34_abf6640f40d37677ef6f809e09046055_3) (Utilities|Struct|BreakFluidForceDynamic _forceinfo_2_a6a35e3243faf59d161a5fbaa2f6f2b1_2))
  (bind (_forceinfo_2_a6a35e3243faf59d161a5fbaa2f6f2b1_3 _componentlocation_13_959307184c8e5ccaca55fc8378d92cfd_2 _boundsradius_30_adff818743be39ac4a481d995cb50d03_4 _waterlevel_34_a4e505d148073b883ca7b1b09a3e34a8_4) (Utilities|Struct|BreakFluidForceDynamicPerInstanceData (LocalVariables|GetLocalForce)))
  (Rendering|ClearRenderTarget2D (Variables|Debug|GetForcesRT))
  (Utilities|FlowControl|Sequence
    (:then_0
      (bind _values (Utilities|Map|Values (Variables|Forces|GetDynamicForces)))
      (for _array_element _values
        (LocalVariables|SetLocalForce _array_element)
        (switch Utilities|FlowControl|Switch|SwitchonFluidDynamicForceMeshType (Utilities|Struct|BreakFluidForceDynamic (Utilities|Struct|BreakFluidForceDynamicPerInstanceData (LocalVariables|GetLocalForce)))
          (:NewEnumerator0
            (Utilities|IsValid _forcecomponent_34_abf6640f40d37677ef6f809e09046055
              (:Is%20Valid
                (Utilities|Map|Add (Variables|Forces|GetDynamicForces) _forcecomponent_34_abf6640f40d37677ef6f809e09046055 (Utilities|Struct|MakeFluidForceDynamicPerInstanceData _forceinfo_2_a6a35e3243faf59d161a5fbaa2f6f2b1 (Collision|GetComponentBounds _forcecomponent_34_abf6640f40d37677ef6f809e09046055) (- (Collision|GetComponentBounds _forcecomponent_34_abf6640f40d37677ef6f809e09046055) _componentlocation_13_959307184c8e5ccaca55fc8378d92cfd) _boundsradius_30_adff818743be39ac4a481d995cb50d03 _waterlevel_34_a4e505d148073b883ca7b1b09a3e34a8 (Utilities|Map|MakeMap "None" "0, 0, 0") (Utilities|Map|MakeMap "None" "0, 0, 0"))))))
          (:NewEnumerator1
            (Utilities|IsValid _forcecomponent_34_abf6640f40d37677ef6f809e09046055_3
              (:Is%20Valid
                (bind _asskeletal_mesh_component (Utilities|Casting|CastToSkeletalMeshComponent _forcecomponent_34_abf6640f40d37677ef6f809e09046055_3))
                (bind _keys (Utilities|Map|Keys _socketvelocitymap_26_82b0e24b45935a12e1949f918a59a537_2))
                (for _array_element_2 _keys
                  (LocalVariables|SetCurSocketLoc (Transformation|GetSocketLocation _asskeletal_mesh_component _array_element_2))
                  (Utilities|Map|Add _socketvelocitymap_26_82b0e24b45935a12e1949f918a59a537 _array_element_2 (- (LocalVariables|GetCurSocketLoc) (Utilities|Map|Find _socketlocationmap_21_abf6aa244a5f84728a5e83be2328c7fa_2 _array_element_2)))
                  (Utilities|Map|Add _socketlocationmap_21_abf6aa244a5f84728a5e83be2328c7fa _array_element_2 (LocalVariables|GetCurSocketLoc)))
                (Utilities|Struct|SetmembersinFluidForceDynamicPerInstanceData (LocalVariables|GetLocalForce) (Collision|GetComponentBounds (Utilities|Struct|BreakFluidForceDynamic (Utilities|Struct|BreakFluidForceDynamicPerInstanceData (LocalVariables|GetLocalForce)))))
                (Utilities|Map|Add (Variables|Forces|GetDynamicForces) _forcecomponent_34_abf6640f40d37677ef6f809e09046055_2 (LocalVariables|GetLocalForce))))))))
    (:then_1
      ;[1 1 1] Update Projectiles Separately
      (bind _values_2 (Utilities|Map|Values (Variables|Forces|GetProjectileForces)))
      (for _array_element_3 _values_2
        (LocalVariables|SetLocalForce _array_element_3)
        (switch Utilities|FlowControl|Switch|SwitchonFluidDynamicForceMeshType (Utilities|Struct|BreakFluidForceDynamic (Utilities|Struct|BreakFluidForceDynamicPerInstanceData (LocalVariables|GetLocalForce)))
          (:NewEnumerator0
            (Utilities|Map|Add (Variables|Forces|GetProjectileForces) (Utilities|Struct|BreakFluidForceDynamic _forceinfo_2_a6a35e3243faf59d161a5fbaa2f6f2b1_3) (Utilities|Struct|MakeFluidForceDynamicPerInstanceData _forceinfo_2_a6a35e3243faf59d161a5fbaa2f6f2b1_3 (Math|Vector|SelectVector (Collision|GetComponentBounds (Utilities|Struct|BreakFluidForceDynamic _forceinfo_2_a6a35e3243faf59d161a5fbaa2f6f2b1_3)) _componentlocation_13_959307184c8e5ccaca55fc8378d92cfd_2 (Utilities|IsValid (Utilities|Struct|BreakFluidForceDynamic _forceinfo_2_a6a35e3243faf59d161a5fbaa2f6f2b1_3) :!class /Script/Engine.KismetSystemLibrary)) (- (Math|Vector|SelectVector (Collision|GetComponentBounds (Utilities|Struct|BreakFluidForceDynamic _forceinfo_2_a6a35e3243faf59d161a5fbaa2f6f2b1_3)) _componentlocation_13_959307184c8e5ccaca55fc8378d92cfd_2 (Utilities|IsValid (Utilities|Struct|BreakFluidForceDynamic _forceinfo_2_a6a35e3243faf59d161a5fbaa2f6f2b1_3) :!class /Script/Engine.KismetSystemLibrary)) _componentlocation_13_959307184c8e5ccaca55fc8378d92cfd_2) _boundsradius_30_adff818743be39ac4a481d995cb50d03_4 _waterlevel_34_a4e505d148073b883ca7b1b09a3e34a8_4 (Utilities|Map|MakeMap "None" "0, 0, 0") (Utilities|Map|MakeMap "None" "0, 0, 0")))))))))

(fn UpdateImpulseLifetimes
  (Utilities|Array|Clear (LocalVariables|GetLocalForces))
  (for _array_element (Variables|Forces|GetImpulseForces)
    (bind _structout (Utilities|Struct|SetmembersinFluidForceImpulsePerInstanceData _array_element (+ (.ElapsedTime_44_12C387C1450456E47FC74BBD11EEAE4F (Utilities|Struct|BreakFluidForceImpulsePerInstanceData _array_element)) (Utilities|Time|GetWorldDeltaSeconds))))
    (if (> (Utilities|Struct|BreakFluidForceImpulse (Utilities|Struct|BreakFluidForceImpulsePerInstanceData _array_element)) (- (.ElapsedTime_44_12C387C1450456E47FC74BBD11EEAE4F (Utilities|Struct|BreakFluidForceImpulsePerInstanceData _array_element)) (.StartOffset_48_78EF0E3F4B05F70C5679F9B9F41D590F (Utilities|Struct|BreakFluidForceImpulsePerInstanceData _array_element))))
      (Utilities|Array|Add (LocalVariables|GetLocalForces) _structout)))
  (Variables|Forces|SetImpulseForces (LocalVariables|GetLocalForces)))

(fn ConstructionScript
  (bind _fluid_size (Variables|Simulation|GetFluid%20Size))
  (bind _returnvalue (/ _fluid_size 256.0))
  (bind _returnvalue_2 (Math|Conversions|ToVector(Float) _returnvalue))
  (Variables|Default|SetFixedStep (/ 1 (Variables|Simulation|GetVirtual%20FPS)))
  (Variables|Default|SetFluidSizeSquared (* (Variables|Simulation|GetFluid%20Size) (Variables|Simulation|GetFluid%20Size)))
  ;[0.7292 0.053 0.06751] Debug Water Plane
  (bind _returnvalue_3 (Rendering|Material|CreateDynamicMaterialInstance "/FortWater/Materials/Debug/M_DebugWaterPlane.M_DebugWaterPlane" "None" "None" :!class /Script/Engine.KismetMaterialLibrary))
  (bind _output_get (Variables|Debug|SetDebugWaterPlaneMID _returnvalue_3))
  (Rendering|Material|SetTextureParameterValue _output_get "Texture" (Variables|Debug|GetForcesRT))
  (Rendering|Material|SetScalarParameterValue _output_get "DisplayNormal" 0.0 :!class /Script/Engine.MaterialInstanceDynamic)
  (Rendering|Material|SetMaterial (Variables|Default|GetDebugPlane) 0 _output_get)
  (CallFunction|SetShowDebugWaterPlane (Variables|Debug|GetShowDebugWaterPlane))
  (if (Variables|Simulation|GetShow%20Simulation%20Mesh)
    (Utilities|FlowControl|Sequence
      (:then_0
        (bind _returnvalue_4 (AddComponent|NODE_AddStaticMeshComponent-0))
        (bind _output_get_2 (Variables|Default|SetFluid%20Display%20Mesh _returnvalue_4))
        (Transformation|SetWorldScale3D _output_get_2 _returnvalue_2)
        ;[0.2907 0.6459 1] Setup Cross Section Display
        (if (Variables|Debug|GetShow%20Cross%20Section)
          (bind _returnvalue_5 (AddComponent|NODE_AddStaticMeshComponent-1 :RelativeTransform (Math|Transform|MakeTransform "0, 0, 0" "90.000000, 0,0.000000" _returnvalue_2)))
          (Variables|Default|SetCross%20Section%20Mesh _returnvalue_5)
          (Transformation|SetWorldLocation (Variables|Default|GetCross%20Section%20Mesh) (Math|Vector|MakeVector (.x (Math|Transform|TransformLocation (Transformation|GetActorTransform) (Variables|Debug|GetCutPos))) (.y (Transformation|GetActorLocation)) (- (.z (Transformation|GetActorLocation)) (- (* (Variables|Simulation|GetFluid%20Size) 0.5) 8.0))) false false)))
      (:then_1
        (Default|SetupDisplayMIDs)))))

(fn ValidateRTs
  (LocalVariables|SetAnyRTInvalid false)
  (for (_array_element _array_index) (Variables|Debug|GetRippleRTs)
    (Utilities|IsValid _array_element
      (:Is%20Not%20Valid
        (LocalVariables|SetAnyRTInvalid true)
        (Water|PrinttoWaterLog (WorldSubsystems|GetWaterSubsystem) (Utilities|String|Append "WATER LOG: " "Sim RT Index: " (Utilities|String|ToString(Integer) _array_index) "was invalid. Fluidsim will fail.") false))))
  (Utilities|FlowControl|Sequence
    (:then_0
      (Utilities|IsValid (Variables|Debug|GetNormalRT)
        (:Is%20Not%20Valid
          (LocalVariables|SetAnyRTInvalid true)
          (Water|PrinttoWaterLog (WorldSubsystems|GetWaterSubsystem) (Utilities|String|Append "WATER LOG: " "NormalRT" "" "was invalid. Fluidsim will fail.") false))))
    (:then_1
      (Utilities|IsValid (Variables|Debug|GetForcesRT)
        (:Is%20Not%20Valid
          (LocalVariables|SetAnyRTInvalid true)
          (Water|PrinttoWaterLog (WorldSubsystems|GetWaterSubsystem) (Utilities|String|Append "WATER LOG: " "Forces RT " "" "was invalid. Fluidsim will fail.") false))))
    (:then_2
      (return (not (LocalVariables|GetAnyRTInvalid))))))
