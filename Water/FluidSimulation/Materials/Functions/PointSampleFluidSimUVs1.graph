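# Material function that snaps UVs to texel centers of the fluid simulation grid, giving point-sampled lookups.
# It scales the default texture coordinates by the FluidSimResolution parameter from MPC_Water, floors them, adds a half-texel offset and divides by the resolution.
# It takes no inputs and outputs the snapped UVs.
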MaterialFunction /Water/FluidSimulation/Materials/Functions/PointSampleFluidSimUVs1
graph {nodes [
  n1 = CollectionParameter() {collection=/Water/Materials/MPC/MPC_Water, parameterName=FluidSimResolution}
  n2 = Floor(None=Multiply(A=TextureCoordinate(), B=@n1))
  FunctionOutput(None=Divide(A=Add(A=@n2) {constB=0.5}, B=@n1))
 ]}
