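# Material function that converts a world position into UV coordinates for the water wave foam simulation area.
# It takes the XY of the input position, defaulting to the pixel's world position, and subtracts the SimLocation parameter from the MPC_Water collection.
# The offset is divided by WaveCaptureSize from the same collection and shifted by half, so the simulation location maps to the UV center.
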MaterialFunction /Water/FluidSimulation/Materials/Functions/WaveFoamSim_WorldPosition_to_UV
graph {
 properties {libraryCategoriesText [Water]}
 nodes [
  n1 = WorldPosition()
  n2 = Subtract(
   A=ComponentMask(None=FunctionInput(Preview=@n1.XYZ) {bUsePreviewValueAsDefault=true, inputName=WorldPosition, inputType=FunctionInput_Vector2}) {g=true, r=true}
   B=ComponentMask(None=CollectionParameter() {collection=/Water/Materials/MPC/MPC_Water, parameterName=SimLocation}) {g=true, r=true}
  )
  n3 = CollectionParameter() {collection=&MPC_Water, parameterName=WaveCaptureSize}
  FunctionOutput(None=Add(A=Divide(A=@n2, B=@n3)) {constB=0.5}) {outputName=UV}
 ]
}
