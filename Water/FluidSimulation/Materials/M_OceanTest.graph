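# M_OceanTest is a lit test material for previewing the area of a player-centered water fluid simulation.
# World XY relative to PlayerLocation from MPC_Water, scaled by FluidSize, becomes clamped UVs.
# Those UVs sample a texture, currently a Black placeholder, whose RG derives a normal.
# A debug term adds a sphere mask at the player to a dim inverted box mask marking the window's edges, alongside color and roughness parameters.
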Material /FortWater/FluidSimulation/Materials/M_OceanTest
graph {
 properties {bUsedWithStaticLighting=true, pixelDepthOffsetMode=PDOM_Legacy, usedShadingModels=Default Lit}
 nodes [
  n1 = WorldPosition()
  n2 = CollectionParameter() {collection=/Water/Materials/MPC/MPC_Water, parameterName=PlayerLocation}
  n3 = Subtract(
   A=ComponentMask(None=WorldPosition().XYZ) {g=true, r=true}
   B=ComponentMask(None=CollectionParameter() {collection=&MPC_Water, parameterName=PlayerLocation}) {g=true, r=true}
  )
  n4 = ScalarParameter() {defaultValue=512, parameterName=FluidSize}
  n5 = Saturate(None=Add(A=Divide(A=@n3, B=@n4)) {constB=0.5})
  n6 = MaterialFunctionCall!(
   A=@n5
   B=Constant() {r=0.5}
   Bounds=ScalarParameter() {defaultValue=0.9, parameterName=bs}
   Edge Falloff=ScalarParameter() {defaultValue=0.05}
  ) {materialFunction=/Engine/Functions/Engine_MaterialFunctions01/Gradient/BoxMask-2D}
  Add(
   A=SphereMask(
    A=ComponentMask(None=@n1.XYZ) {g=true, r=true}
    B=ComponentMask(None=@n2) {g=true, r=true}
    Radius=ScalarParameter() {defaultValue=50, parameterName=DebugRadius}
   ) {hardnessPercent=0}
   B=Multiply(A=OneMinus(None=@n6.Result)) {constB=0.2}
  )
  DeriveNormalZ(InXY=ComponentMask(None=TextureSample(UVs=@n5) {texture=&Black}.RGB) {g=true, r=true})
  ScalarParameter() {parameterName=Roughness}
  VectorParameter() {parameterName=B, defaultValue=<rgba 0 0.05 0.1 0>}
 ]
}
