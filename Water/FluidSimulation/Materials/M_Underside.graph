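# Unlit masked material for the underside of the fluid simulation's water surface, faking reflection and refraction with a sunset ambient cubemap.
# Normals rebuilt from a NormalAndHeight texture drive two cubemap samples: a Fresnel-weighted reflection off the flipped normal and a view-bent lookup weighted by inverse Fresnel.
# The texture's blue channel scales a height vector, and a U gradient minus a Cut parameter forms a clipping term.
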Material /FortWater/FluidSimulation/Materials/M_Underside
graph {
 properties {
  bScreenSpaceReflections=true
  blendMode=BLEND_Masked
  opacityMaskClipValue=0.001
  pixelDepthOffsetMode=PDOM_Legacy
  shadingModel=MSM_Unlit
  translucencyLightingMode=TLM_Surface
  usedShadingModels=Unlit
 }
 nodes [
  Add(
   A=Subtract(
    A=MaterialFunctionCall() {materialFunction=&LinearGradient}.UGradient
    B=ScalarParameter() {parameterName=Cut}
   )
   B=ScalarParameter() {defaultValue=0.001, parameterName=bias}
  )
  n1 = TextureSampleParameter2D() {parameterName=NormalAndHeight, texture=&Black}
  n2 = DeriveNormalZ(InXY=ComponentMask(None=@n1.RGB) {g=true, r=true})
  n3 = Multiply(A=@n2) {constB=-1}
  n4 = TextureSample(UVs=Multiply(A=ReflectionVectorWS(CustomWorldNormal=@n3)) {constB=-1}) {texture=/Engine/MapTemplates/Sky/SunsetAmbientCubemap}
  n5 = ScalarParameter() {defaultValue=3, parameterName=Bottom Reflect Power}
  n6 = LinearInterpolate(A=@n2, B=Multiply(A=CameraVectorWS()) {constB=-1}, Alpha=ScalarParameter() {defaultValue=0.2, parameterName=ref})
  n7 = OneMinus(None=Fresnel(BaseReflectFractionIn=ScalarParameter() {parameterName=BaseOpacity}) {exponent=1})
  n8 = ScalarParameter() {defaultValue=1, parameterName=Exp}
  Add(
   A=Multiply(A=Multiply(A=@n4.RGB) {constB=0.5}, B=Fresnel(ExponentIn=@n5) {exponent=1})
   B=Multiply(A=TextureSample(UVs=@n6) {texture=&SunsetAmbientCubemap}.RGB, B=Power(Base=@n7, Exp=@n8))
  )
  Multiply(A=@n1.B, B=VectorParameter() {parameterName=H, defaultValue=<rgba 0 0 1 0>}.RGB)
  ScalarParameter()
 ]
}
