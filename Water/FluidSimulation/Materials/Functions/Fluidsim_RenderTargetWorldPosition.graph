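# Material function that converts fluid simulation render target UVs into a 2D world-space XY position for the water fluid sim.
# It centers the UV input, which defaults to texture coordinates, around zero, scales it by FluidSimSize and offsets it by SimLocation.
# Both parameters come from the MPC_Water parameter collection, so the mapping follows the simulation's size and location set there.
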MaterialFunction /Water/FluidSimulation/Materials/Functions/Fluidsim_RenderTargetWorldPosition
graph {nodes [
  n1 = Subtract(A=FunctionInput(Preview=TextureCoordinate()) {bUsePreviewValueAsDefault=true, inputName=UV, inputType=FunctionInput_Vector2}) {constB=0.5}
  n2 = ComponentMask(None=CollectionParameter() {collection=/Water/Materials/MPC/MPC_Water, parameterName=FluidSimSize}) {g=true, r=true}
  n3 = CollectionParameter() {collection=&MPC_Water, parameterName=SimLocation}
  FunctionOutput(None=Add(A=Multiply(A=@n1, B=@n2), B=ComponentMask(None=@n3) {g=true, r=true})) {outputName=Fluidsim WorldPosition}
 ]}
