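# Material function that outputs texture coordinates snapped to the texel centres of a fluid simulation texture, so the sim reads as point sampled.
# It scales the UVs by the XY of the SimResolution vector parameter and floors them to a texel index.
# It then adds a half-texel offset and divides by the resolution to return to UV space.
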MaterialFunction /FortWater/FluidSimulation/Materials/Functions/PointSampleFluidSimUVs
graph {nodes [
  n1 = ComponentMask(None=VectorParameter() {parameterName=SimResolution, defaultValue=<rgba 512 512 0 0>}.RGB) {g=true, r=true}
  n2 = Floor(None=Multiply(A=TextureCoordinate(), B=@n1))
  FunctionOutput(None=Divide(A=Add(A=@n2) {constB=0.5}, B=@n1))
 ]}
